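; Gameplay ability that transforms the player into the Beast Girl pawn through a standing or swimming montage.
; After a cost check, the server spawns the stand-in pawn and hides the original, while movement is blocked and weapons holstered.
; Montage notifies drive audio cues and the transform, which applies durability cost, commits and possesses the new pawn.
; Ending cleans up, restoring the original pawn and destroying the stand-in if unfinished.
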(event Ability|EventActivateAbility
  (bind _returnvalue (Ability|GetAvatarActorfromActorInfo))
  (bind _returnvalue_2 (Transformation|GetActorLocation _returnvalue))
  (bind _iswatertransformation (Variables|Default|GetIsWaterTransformation))
  (bind _swimmingtransformationmontage (Variables|Default|GetSwimmingTransformationMontage))
  (bind _standingtransformationmontage (Variables|Default|GetStandingTransformationMontage))
  (bind _returnvalue_3 (select _iswatertransformation _swimmingtransformationmontage _standingtransformationmontage))
  (Variables|Default|SetFinishedTransformation false)
  (bind _returnvalue_4 (Ability|CheckAbilityCost))
  (if _returnvalue_4
    (Variables|Default|SetIsWaterTransformation (or (WaterBody|IsSurfaceSwimming (CallFunction|GetAvatarasFortPlayerPawn)) (Water|IsWaterSprintBoosting (CallFunction|GetAvatarasFortPlayerPawn))))
    (Ability|HasAuthority
      (:True
        (bind _returnvalue_5 (Class|BeastGirlStandinPawn|SpawnBeastGirlforPlayerPawn (CallFunction|GetAvatarasFortPlayerPawn) "/BeastGirlStormClawsGameplay/Gameplay/Actors/BP_BeastGirlPawn.BP_BeastGirlPawn_C"))
        (bind _output_get (Variables|Default|SetBeastGirlPawn _returnvalue_5))
        (Affiliation|SetAffiliationGameplayOwnership (CallFunction|GetAvatarasFortPlayerPawn) _output_get)
        (Class|BeastGirlStandinPawn|SetIsTransformingInto _output_get true)
        (Class|FortStandinPlayerPawn|HideandAttachReplacedPawn _output_get (CallFunction|GetAvatarasFortPlayerPawn))
        (Class|BeastGirlStandinPawn|AttachOriginalPawntoMovementBase _output_get)
        (Class|BeastGirlTransformationManager|SetBeastGirlPawn (Actor|GetComponentbyClass (Ability|GetPlayerControllerfromActorInfo) "/Script/BeastGirlRuntime.BeastGirlTransformationManager") _output_get))
      (:False
        (bind _asfort_player_controller (Utilities|Casting|CastToFortPlayerController (Ability|GetPlayerControllerfromActorInfo))
          (:then
            (Class|FortPlayerController|CacheCurrentPrimaryBarSlottedItems _asfort_player_controller))
          (:CastFailed
            _))))
    (Development|PrintString (Utilities|GetDisplayName _returnvalue_3))
    (bind (_asynctaskproxy _notifytag) (Ability|Tasks|PlayMontageAndWaitNotify "PlayMontage" _returnvalue_3 1.0 "Transform" false)
      (:then
        (Variables|Default|SetAsync%20Task _asynctaskproxy)
        (bind _returnvalue_6 (Ability|ApplyGameplayEffectToOwner "/BeastGirlStormClawsGameplay/Gameplay/Abilities/GE_Transformation.GE_Transformation_C"))
        (Variables|Default|SetGEPreventMoveHandle _returnvalue_6)
        (Weapon|SetHolsterWeapon (CallFunction|GetAvatarasFortPlayerPawn) (CallFunction|GetAvatarasFortPlayerPawn) true false (Variables|Default|GetHolster%20Id)))
      (:OnCompleted
        _)
      (:OnCancelled
        (bind _asfort_player_controller_2 (Utilities|Casting|CastToFortPlayerController (Ability|GetPlayerControllerfromActorInfo)))
        (CameraShakes|StopAllInstancesOfCameraShake (Variables|/Script/Engine.PlayerController|GetPlayerCameraManager _asfort_player_controller_2) "/BeastGirlStormClawsGameplay/Assets/Animation/Camera/CheerTransform_Camera_BP.CheerTransform_Camera_BP_C" true)
        (Ability|EndAbility))
      (:OnNotifyBegin
        _))
    (if (== _notifytag "(TagName=\"Athena.Gameplay.BentBaton.Transform\")")
      (if (Variables|Default|GetFinishedTransformation)
        _)
      (elif (== _notifytag "(TagName=\"Athena.Gameplay.BentBaton.Transform.AudioCue\")")
        (Ability|ExecuteGameplayCueOnOwner (Variables|Default|GetGCN_Transform) (Utilities|Struct|MakeGameplayEffectContextHandle))
        (Audio|PlaySoundatLocation "/BeastGirlStormClawsGameplay/Assets/Audio/SoundCues/Abilities/SC_BeastGirlStormClaws_AOE_Charge.SC_BeastGirlStormClaws_AOE_Charge" (Transformation|GetActorLocation (Ability|GetAvatarActorfromActorInfo)))))
    (Audio|PlaySoundatLocation "/BeastGirlStormClawsGameplay/Assets/Audio/SoundCues/Vox/SC_BeastGirlStormClaws_Activate_VO.SC_BeastGirlStormClaws_Activate_VO" _returnvalue_2 "0, 0, 0" 2.5)
    (Variables|Default|SetFinishedTransformation true)
    (Class|BeastGirlTransformationManager|ApplyConsumableDurabilityCost (Actor|GetComponentbyClass (Ability|GetPlayerControllerfromActorInfo) "/Script/BeastGirlRuntime.BeastGirlTransformationManager"))
    (Ability|CommitAbility)
    (CallFunction|HandleTransform)
    (Audio|PlaySoundatLocation "/BeastGirlStormClawsGameplay/Assets/Audio/SoundCues/Abilities/SC_BeastGirlStormClaws_AOE_Burst.SC_BeastGirlStormClaws_AOE_Burst" _returnvalue_2)))

(event Ability|EventOnEndAbility
  (params (bWasCancelled :type Boolean))
  (bind _beastgirlpawn (Variables|Default|GetBeastGirlPawn))
  (Development|PrintString "End")
  (Ability|Animation|MontageStop -1.0)
  (GameplayTasks|EndTask (Variables|Default|GetAsync%20Task))
  (Ability|RemoveGameplayEffectFromOwnerWithHandle (Variables|Default|GetGEPreventMoveHandle))
  (Weapon|SetHolsterWeapon (CallFunction|GetAvatarasFortPlayerPawn) (CallFunction|GetAvatarasFortPlayerPawn) false false (Variables|Default|GetHolster%20Id))
  (Utilities|IsValid _beastgirlpawn
    (:Is%20Valid
      (Class|BeastGirlStandinPawn|DetachOriginalPawnfromMovingBase _beastgirlpawn)
      (if (not (Variables|Default|GetFinishedTransformation))
        (bind _asfort_player_controller (Utilities|Casting|CastToFortPlayerController (Ability|GetPlayerControllerfromActorInfo)))
        (Class|BeastGirlStandinPawn|RestoretoOriginalPawn _beastgirlpawn _asfort_player_controller)
        (Actor|DestroyActor _beastgirlpawn)))))

(event Custom|HandleTransform
  (Development|PrintString "Handle Transform Reached")
  (Utilities|IsValid (Variables|Default|GetBeastGirlPawn)
    (:Is%20Valid
      (Class|BeastGirlStandinPawn|ReplaceandPossess (Variables|Default|GetBeastGirlPawn) (CallFunction|GetAvatarasFortPlayerPawn)))))

(fn ShouldPlayFailedMontage
  (params (FailedReason :type Gameplay%20Tag%20Container%20Structure%20%28by%20ref%29))
  (return (Utilities|Array|ContainsItem (GameplayTags|BreakGameplayTagContainer FailedReason) "(TagName=\"MovementMode.Falling\")")))
