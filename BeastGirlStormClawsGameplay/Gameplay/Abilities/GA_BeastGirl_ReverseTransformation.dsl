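; Gameplay ability that reverts a Beast Girl stand-in pawn to the original player pawn.
; On activation it applies transformation and swap-block effects, then polls until the pawn is neither falling nor delay-tagged, backed by a failsafe timer.
; Execution restores health and shields, fires an optional cue, and replicates land or water animation data from the server.
; Animation length or timeout ends the ability, which swaps pawns, returns the consumable and clears effects.
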(event Ability|EventActivateAbility
  (CallFunction|ReverseTransformation)
  (bind _returnvalue (Utilities|Time|SetTimerbyEvent (EventDispatchers|CreateEvent :!event Custom|ExecuteTransformation) (ScalableFloat|GetValueatLevel (Variables|Default|GetHF_FailsafeWaitDuration))))
  (Variables|Default|SetFailsafeTimer _returnvalue))

(event Custom|CheckIfTransformationAllowed
  (if (not (CallFunction|ShouldDelayTransformation))
    (CallFunction|ExecuteTransformation)))

(event Custom|ReverseTransformation
  ;[1 1 1] Block Inputs and movement
  (bind _returnvalue (Ability|ApplyGameplayEffectToOwner (Variables|Default|GetTransformationEffect)))
  (Variables|Default|SetTransformationHandle _returnvalue)
  (bind _returnvalue_2 (Ability|ApplyGameplayEffectToOwner (Variables|Default|GetBlockSwap)))
  (Variables|Default|SetBlockSwapHandle _returnvalue_2)
  (if (CallFunction|ShouldDelayTransformation)
    (bind _returnvalue_3 (Utilities|Time|SetTimerbyEvent (EventDispatchers|CreateEvent :!event Custom|CheckIfTransformationAllowed) 0.1 true true))
    (Variables|Default|SetCheckTransformationTimer _returnvalue_3)
    (else
      (CallFunction|ExecuteTransformation))))

(event Custom|ExecuteTransformation
  (bind _returnvalue (Ability|GetPlayerControllerfromActorInfo))
  (bind _returnvalue_2 (CallFunction|GetAvatarasFortPlayerPawn))
  (bind _returnvalue_3 (WaterBody|IsSurfaceSwimming _returnvalue_2))
  (bind _returnvalue_4 (Water|IsWaterSprintBoosting _returnvalue_2))
  (bind _returnvalue_5 (or _returnvalue_3 _returnvalue_4))
  (Utilities|Time|ClearandInvalidateTimerbyHandle (Variables|Default|GetCheckTransformationTimer))
  (Ability|SendGameplayEvent "(TagName=\"BeastGirl.Event.BeginReverseTransformation\")")
  (Utilities|IsValid _returnvalue
    (:Is%20Valid
      (bind _output_get (Variables|Default|SetTransfomationManager (Actor|GetComponentbyClass _returnvalue "/Script/BeastGirlRuntime.BeastGirlTransformationManager")))
      (Class|BeastGirlTransformationManager|RestoreOriginalHealthandShields _output_get)
      (Class|BeastGirlTransformationManager|OldCodeAddConsumablewithRemainingHealth _output_get))
    (:Is%20Not%20Valid
      _))
  (Utilities|Time|SetTimerbyEvent (delegate Custom|TimeOutTransformation) (ScalableFloat|GetValueatLevel (Variables|Default|GetHF_TransfomationApplyTime)))
  (if (ScalableFloat|GetValueasBool (Variables|Default|GetHF_DetransformGC))
    (GameplayCue|ExecuteGameplayCue (CallFunction|GetAvatarasFortPlayerPawn) (Variables|Default|GetGC_Detransform))
    (else
      _))
  (bind _asbeast_girl_stand_in_pawn (Utilities|Casting|CastToBeastGirlStandInPawn _returnvalue_2))
  (Ability|HasAuthority
    (:True
      (Class|BeastGirlStandinPawn|SetReplicatedTransformationData _asbeast_girl_stand_in_pawn (Variables|/Script/BeastGirlRuntime.BeastGirlTransformationManager|GetbOptingOut (Variables|Default|GetTransfomationManager)) _returnvalue_5)
      (bind (_replicatedtransformationdata_btransforming _replicatedtransformationdata_bisoptout) (Utilities|Struct|BreakBeastGirlTransformationData (Variables|/Script/BeastGirlRuntime.BeastGirlStandInPawn|GetReplicatedTransformationData_bTransforming _asbeast_girl_stand_in_pawn)))
      (Utilities|IsValid (select _returnvalue_5 (.InWater (Utilities|Struct|BreakBeastGirlTransformationAnimations (Class|BeastGirlStandinPawn|GetTransformationAnimations _asbeast_girl_stand_in_pawn _replicatedtransformationdata_bisoptout))) (Utilities|Struct|BreakBeastGirlTransformationAnimations (Class|BeastGirlStandinPawn|GetTransformationAnimations _asbeast_girl_stand_in_pawn _replicatedtransformationdata_bisoptout)))
        (:Is%20Valid
          (Utilities|FlowControl|Delay (Animation|GetPlayLength (select _returnvalue_5 (.InWater (Utilities|Struct|BreakBeastGirlTransformationAnimations (Class|BeastGirlStandinPawn|GetTransformationAnimations _asbeast_girl_stand_in_pawn _replicatedtransformationdata_bisoptout))) (Utilities|Struct|BreakBeastGirlTransformationAnimations (Class|BeastGirlStandinPawn|GetTransformationAnimations _asbeast_girl_stand_in_pawn _replicatedtransformationdata_bisoptout))))))
        (:Is%20Not%20Valid
          _))
      (Ability|EndAbility))))

(event Custom|TimeOutTransformation
  (Ability|EndAbility))

(event Ability|EventOnEndAbility
  (params (bWasCancelled :type Boolean))
  (bind _transfomationmanager (Variables|Default|GetTransfomationManager))
  (bind _asbeast_girl_stand_in_pawn (Utilities|Casting|CastToBeastGirlStandInPawn (Ability|GetAvatarActorfromActorInfo)))
  (Parachute|EndSkydiving _asbeast_girl_stand_in_pawn)
  (bind _asfort_player_pawn (Utilities|Casting|CastToFortPlayerPawn (Class|FortStandinPlayerPawn|GetReplacedPawn _asbeast_girl_stand_in_pawn))
    (:then
      (Parachute|EndSkydiving _asfort_player_pawn))
    (:CastFailed
      _))
  (Class|BeastGirlStandinPawn|RestoretoOriginalPawn _asbeast_girl_stand_in_pawn 0)
  (Utilities|IsValid _transfomationmanager
    (:Is%20Valid
      (Class|BeastGirlTransformationManager|NewCodeAddConsumablewithRemainingHealth _transfomationmanager))
    (:Is%20Not%20Valid
      _))
  (Ability|RemoveGameplayEffectFromOwnerWithHandle (Variables|Default|GetBlockSwapHandle))
  (Ability|RemoveGameplayEffectFromOwnerWithHandle (Variables|Default|GetTransformationHandle))
  (Ability|RemoveGameplayEffectFromOwnerWithAssetTags "(GameplayTags=((TagName=\"Granted.BeastGirl.OutOfHealth\")))")
  (CameraShakes|StopAllCameraShakes (Variables|/Script/Engine.PlayerController|GetPlayerCameraManager (Ability|GetPlayerControllerfromActorInfo)) true)
  (Utilities|Time|ClearandInvalidateTimerbyHandle (Variables|Default|GetFailsafeTimer)))

(fn ShouldDelayTransformation
  (bind _output_get (LocalVariables|SetPlayerPawn (CallFunction|GetAvatarasFortPlayerPawn)))
  (Utilities|IsValid _output_get
    (:Is%20Valid
      (return (or (AI|Components|NavMovement|IsFalling (Variables|/Script/Engine.Character|GetCharacterMovement (LocalVariables|GetPlayerPawn))) (GameplayTags|HasAnyMatchingGameplayTags (LocalVariables|GetPlayerPawn) (Variables|Default|GetTagsToDelayTransfromation)))))
    (:Is%20Not%20Valid
      (return false))))
