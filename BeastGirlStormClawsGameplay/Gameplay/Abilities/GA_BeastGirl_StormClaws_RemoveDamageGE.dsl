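; Gameplay ability for BeastGirl's Storm Claws that removes the gameplay effect that granted it once the damage zone tag is gone.
; On activation it waits for the Granted.BeastGirl.StormClaws.InDamageZone tag to be removed, then ends itself.
; Its end handler removes the granting effect, whether or not the ability was cancelled.
; A debug print fires when the wait task starts.
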(event Ability|EventActivateAbility
  (Ability|Tasks|WaitGameplayTagRemove "(TagName=\"Granted.BeastGirl.StormClaws.InDamageZone\")"
    (:then
      (Development|PrintString "Flow is Working" true true "(R=1.000000,G=0.170836,B=0.642237,A=1.000000)" 5.0))
    (:Removed
      (Ability|EndAbility))))

(event Ability|EventOnEndAbility
  (params (bWasCancelled :type Boolean))
  (Ability|RemoveGrantedbyEffect))
