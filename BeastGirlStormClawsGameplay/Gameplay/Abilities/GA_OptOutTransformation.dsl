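; Gameplay ability that lets a transformed player reverse their Beast Girl transformation by holding Use.
; On activation it listens for Use pressed and released events; a press starts a timer from a scalable hold time and fires a UI cue carrying that duration.
; Releasing with a valid timer cancels it and removes the cue.
; When the timer completes, it calls StartReverseTransformation on the player controller's transformation manager component and ends the ability.
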(event Ability|EventActivateAbility
  (Ability|Tasks|WaitGameplayEvent "(TagName=\"Event.Use.Pressed\")"
    (:then
      (Ability|Tasks|WaitGameplayEvent "(TagName=\"Event.Use.Released\")"
        (:EventReceived
          (CallFunction|ExitReleased))))
    (:EventReceived
      (CallFunction|ExitPressed))))

(event Custom|ExecuteOptOut
  (bind _returnvalue (Ability|GetPlayerControllerfromActorInfo))
  (bind _returnvalue_2 (Actor|GetComponentbyClass _returnvalue "/Script/BeastGirlRuntime.BeastGirlTransformationManager"))
  (Utilities|IsValid _returnvalue
    (:Is%20Valid
      (Utilities|IsValid _returnvalue_2
        (:Is%20Valid
          (Class|BeastGirlTransformationManager|StartReverseTransformation _returnvalue_2 true)
          (Ability|EndAbility))))))

(event Custom|ExitPressed
  (bind _holdtimeoptout (Variables|Default|GetHoldTimeOptOut))
  (bind _returnvalue (ScalableFloat|GetValueatLevel _holdtimeoptout))
  (bind _returnvalue_2 (Utilities|Time|SetTimerbyEvent (EventDispatchers|CreateEvent :!event Custom|ExecuteOptOut) _returnvalue))
  (Variables|Default|SetOptOutTimer _returnvalue_2)
  (EventDispatchers|CallOnUIGameplayCue (Utilities|Casting|CastToFortPlayerPawn (Ability|GetAvatarActorfromActorInfo)) (GameplayTags|GetTagName (Variables|Default|GetEquipCueTag)) "WhileActive" (Ability|GameplayCue|MakeGameplayCueParameters 0.0 _returnvalue)))

(event Custom|ExitReleased
  (if (Utilities|Time|IsValidTimerHandle (Variables|Default|GetOptOutTimer))
    (Utilities|Time|ClearandInvalidateTimerbyHandle (Variables|Default|GetOptOutTimer))
    (bind _asfort_player_pawn (Utilities|Casting|CastToFortPlayerPawn (Ability|GetAvatarActorfromActorInfo)))
    (EventDispatchers|CallOnUIGameplayCue _asfort_player_pawn (GameplayTags|GetTagName (Variables|Default|GetEquipCueTag)) "Removed" (Ability|GameplayCue|MakeGameplayCueParameters 0.0))))
